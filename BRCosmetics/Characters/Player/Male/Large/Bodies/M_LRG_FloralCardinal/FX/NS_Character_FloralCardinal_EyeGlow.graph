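# Niagara eye glow effect for the FloralCardinal large male player character body.
# One emitter burst-spawns long-lived red sprites that re-sample their location on the skeletal mesh every update, with a camera offset.
# A waveform with a random input offset pulses the color brightness.
# Alpha combines a time-of-day scalar, dimmer at night and evening, with a system-level fade value.
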NiagaraSystem /BRCosmetics/Characters/Player/Male/Large/Bodies/M_LRG_FloralCardinal/FX/NS_Character_FloralCardinal_EyeGlow
dependencies {
 usedDataInterfaces [&NiagaraDataInterfaceSkeletalMesh, &NiagaraDataInterfaceCurve]
 usedDynamicInputs [
  &Multiply_FloatByInt
  &ReturnParticleIDAttributes
  &FloatFromCurve
 ]
 usedModules [
  &self:SystemSpawnScript.SystemScriptSource.SystemScriptGraph.NiagaraNodeAssignment_0.SetVariables_#1
  &SystemState
  &self:SystemSpawnScript.SystemScriptSource.SystemScriptGraph.NiagaraNodeAssignment_1.SetVariables_#2
 ]
}
emitters {EyeGlow {
  emitterUpdateScript [
   &EmitterState(
    Inactive Response=::NewEnumerator1
    Life Cycle Mode=::NewEnumerator1
    Loop Behavior=::NewEnumerator1
    Loop Duration Mode=::NewEnumerator1
    UseLoopDelay=false
   )
   &SpawnBurst_Instantaneous(Spawn Probability=HlslExpression() {hlslExpression=1.0}, Use Spawn Probability=false, UseLoopCountLimit=false)
  ]
  particleSpawnScript [
   &InitializeParticle(
    Color=<rgba 1 0 0.0341 0.5>
    Mesh Scale=<1 0.8 0.8>
    Mesh Scale Mode=::NewEnumerator4
    Sprite Size Mode=::NewEnumerator3
    Uniform Sprite Size=8
    UsePositionOffset=false
    Write Material Random=true
   )
   &InitialMeshOrientation(
    Mesh Axis To Orient=<0 1 0>
    Mesh Orientation Mode=::NewEnumerator2
    Orientation Vector=<0 1 0>
    Rotation=<-0.0228 -0.0407 -0.0845>
    Use Rotation Vector=true
   )
   /SetVariables() {isSetParameters=true}
  ]
  particleUpdateScript [
   &ParticleState(Kill Particles When Lifetime Has Elapsed=false, Let Infinitely Lived Particles Die When Emitter Deactivates=true)
   &CameraOffset(Camera Offset Amount=3)
   &SkeletalMeshLocation(
    MaskSpawnGroup=false
    Mesh Sampling Type=::NewEnumerator5
    Orientation Sampling=::NewEnumerator1
    Sampled Position Offset Coordinate Space:ENiagaraSkelMeshTransforms=::NewEnumerator1
    Skeletal Mesh=$System.Skeletal Mesh
    Socket Sampling Mode=::NewEnumerator5
   ) {module=SkeletalMeshLocation001}
   &ScaleColor(
    Scale Alpha=&Multiply_Float(A=&FNTimeOfDayScalar(Night Scale=0.5, Morning Scale=1, Evening Scale=0.5), B=$System.Fade)
    Scale RGB=&Multiply_VectorByFloat(
     Vector=&MakeVectorFromLinearColorRGB(Color=<rgba 1 0.0191 0 1>)
     Float=&Waveform(
      Global Input Position=&RandomRangeFloat(Minimum=0, Maximum=100)
      "Waveform [1]"=::NewEnumerator0
      "[1] Frequency"=0.5
      "[1] Amplitude Min/Max"=<1 5>
     )
    )
    ScaleA=true
    ScaleRGB=true
   )
  ]
  rendererClasses [&NiagaraSpriteRendererProperties]
  renderers [{rendererClass=&NiagaraSpriteRendererProperties}]
 }}
summary {
 systemName=NS_Character_FloralCardinal_EyeGlow
 emitters [{emitterName=EyeGlow, rendererClasses [&NiagaraSpriteRendererProperties]}]
 userVariables [
  User.bIsFrontEnd {type=NiagaraBool}
  User.bIsFrontEndPreview {type=NiagaraBool}
 ]
}
