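# Parent material for the FloralCardinal character body, adding advanced car-style clear coat to the 2023 character shading stack.
# PreFX output gets a metallic-weighted roughness multiplier, then passes through the clearcoat function, MF_FloralCardinal and PostFX.
# The clearcoat's orange-peel normal, angle-correct blended with itself, drives the clear coat normal output.
# The final surface Normal is restored from PreFX, overriding later changes.
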Material /BRCosmetics/Characters/Player/Male/Large/Bodies/M_LRG_FloralCardinal/Materials/M_AdvancedClearcoat_Parent_FloralCardinal
graph {
 properties {
  bUseMaterialAttributes=true
  bUsedWithMeshDeformer=true
  bUsedWithMorphTargets=true
  bUsedWithSkeletalMesh=true
  pixelDepthOffsetMode=PDOM_Legacy
  shadingModel=MSM_FromMaterialExpression
  usedShadingModels=Default Lit | Subsurface | Clear Coat | Subsurface Profile | Cloth
 }
 nodes [
  n1 = MaterialFunctionCall() {materialFunction=&MF_CharacterParent_2023_PreFX}
  n2 = GetMaterialAttributes(None=@n1.Result) {attributeGetTypes [#1, #2]}
  n3 = Multiply(A=@n2.Roughness, B=ScalarParameter() {defaultValue=1, parameterName=Metal Roughness Multiplier})
  n4 = Saturate(None=Power(Base=@n2.Metallic))
  n5 = MaterialFunctionCall!(
   Mask=@n1.Tech Art Mask
   Fresnel=@n1.Tech Art Mask
   In=SetMaterialAttributes!(MaterialAttributes=@n2.MaterialAttributes, Roughness=LinearInterpolate(A=@n2.Roughness, B=@n3, Alpha=@n4)) {attributeSetTypes [#1]}
  ) {materialFunction=/Game/Global/MaterialFunctions/TechArt/MF_Advanced_CarClearCoat}
  ClearCoatNormalCustomOutput!(None=MaterialFunctionCall(BaseNormal=@n5.NormalWithOrangePeel, AdditionalNormal=@n5.NormalWithOrangePeel) {materialFunction=&BlendAngleCorrectedNormals}.Result)
  SetMaterialAttributes!(
   MaterialAttributes=MaterialFunctionCall(
    Material=MaterialFunctionCall!(MA=@n5.Result) {materialFunction=&/MF_FloralCardinal}.Result
    M=@n1.M
    Pre Skinned Local Position=@n1.M
    Distance Blend=@n1.M
    Fresnel=@n1.M
   ) {materialFunction=&MF_CharacterParent_2023_PostFX}.Result
   Normal=GetMaterialAttributes(None=@n1.Result) {attributeGetTypes [#3]}.Normal
  ) {attributeSetTypes [#3]}
 ]
}
