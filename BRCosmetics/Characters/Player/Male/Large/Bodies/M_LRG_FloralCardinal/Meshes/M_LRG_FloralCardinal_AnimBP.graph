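# Animation Blueprint for the large male FloralCardinal player body cosmetic, copying its pose from the attached parent mesh.
# Bone constraints adjust arm twist, corrective and knee pad bones, and additive offsets shift both pad bones.
# The pose then feeds a character-specific Control Rig that receives an LOD2 flag.
# Each update sets that flag and maps clavicle pitch relative to spine_05 into upper-arm variables, which the shown AnimGraph never reads.
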AnimBlueprint /BRCosmetics/Characters/Player/Male/Large/Bodies/M_LRG_FloralCardinal/Meshes/M_LRG_FloralCardinal_AnimBP
target_skeleton=&/M_LRG_FloralCardinal_Skeleton
graphs [
 AnimGraph {nodes [
   n1 = "Constraint - lowerarm_twist_02_r_corrective (P)"(ComponentPose=Local To Component(LocalPose=Copy Pose From Mesh() {bUseAttachedParent=true}), Alpha=1, ConstraintWeights_0=1) {
    alphaBoolBlend {blendOption=Linear}
    boneToModify {boneName=lowerarm_twist_02_r_corrective}
    constraintSetup [{targetBone {boneName=lowerarm_twist_02_r}}]
    constraintWeights [1]
   }
   n2 = "Constraint - lowerarm_twist_02_l_corrective (P)"(ComponentPose=@n1, Alpha=1, ConstraintWeights_0=1) {
    alphaBoolBlend {blendOption=Linear}
    boneToModify {boneName=lowerarm_twist_02_l_corrective}
    constraintSetup [{targetBone {boneName=lowerarm_twist_02_l}}]
    constraintWeights [1]
   }
   n3 = "Constraint - upperarm_twist_01_r (P)"(ComponentPose=@n2, Alpha=0.1, ConstraintWeights_0=1) {
    alphaBoolBlend {blendOption=Linear}
    boneToModify {boneName=upperarm_twist_01_r}
    constraintSetup [{targetBone {boneName=upperarm_r}}]
    constraintWeights [1]
   }
   n4 = "Transform (Modify) Bone - Bone: pad_L"(ComponentPose=@n3, Alpha=1, Translation="3,0,0", Rotation="5,5,0", Scale="1,1,1") {
    rotationMode=BMM_Additive
    rotationSpace=BCS_BoneSpace
    translationMode=BMM_Additive
    alphaBoolBlend {blendOption=Linear}
    boneToModify {boneName=pad_L}
    rotation {pitch=5, yaw=5}
    translation {x=3}
   }
   n5 = "Transform (Modify) Bone - Bone: pad_R"(ComponentPose=@n4, Alpha=1, Translation="-3.179,0,0", Rotation="-5,0,0", Scale="1,1,1") {
    rotationMode=BMM_Additive
    rotationSpace=BCS_BoneSpace
    translationMode=BMM_Additive
    alphaBoolBlend {blendOption=Linear}
    boneToModify {boneName=pad_R}
    rotation {pitch=-5}
    translation {x=-3.179}
   }
   n6 = "Constraint - KneePad_l (P)"(ComponentPose=@n5, Alpha=1, ConstraintWeights_0=1) {
    alphaBoolBlend {blendOption=Linear}
    boneToModify {boneName=KneePad_l}
    constraintSetup [{targetBone {boneName=calf_l}}]
    constraintWeights [1]
   }
   n7 = "Constraint - KneePad_R (P)"(ComponentPose=@n6, Alpha=1, ConstraintWeights_0=1) {
    alphaBoolBlend {blendOption=Linear}
    boneToModify {boneName=KneePad_R}
    constraintSetup [{targetBone {boneName=calf_r}}]
    constraintWeights [1]
   }
   n8 = Control Rig!(Source=Component To Local(ComponentPose=@n7), Alpha, LOD2:bool=Get LOD2() -> (LOD2:bool)) -> (Pose) {
    alphaBoolBlend {blendOption=Linear}
    controlRigAssetReference {blueprintRigClass=&/M_LRG_FloralCardinal_CtrlRig.M_LRG_FloralCardinal_CtrlRig_C}
   }
   Output Pose(Result=@n8) {layerGroup=None}
  ]}
 EventGraph {nodes [
   Event BlueprintUpdateAnimation()
   > n1 = Sequence()
   > Set LOD2(LOD2:bool="Equal (Integer)"(A=Get CurrentLOD(), B=2)) -> (Output_Get:bool)
   n2 = GetOwnerSkeletalMeshComponent()
   Set Upperarm_L_rot(
    execute=@n1.then_1
    Upperarm_L_rot:double=MapRangeClamped(
     Value=GetDeltaTransformFromRefPose!(self=@n2, BoneName=clavicle_l, BaseName=spine_05) -> (
      ReturnValue_Location:Vector
      ReturnValue_Rotation_Roll:float
      ReturnValue_Rotation_Pitch:float
      ReturnValue_Rotation_Yaw:float
      ReturnValue_Scale:Vector
     ).ReturnValue_Rotation_Pitch
     InRangeA=0
     InRangeB=60
    )
   ) -> (Output_Get:double)
   n3 = GetOwnerSkeletalMeshComponent()
   > Set Upperarm_R_rot(Upperarm_R_rot:double=MapRangeClamped(Value=GetDeltaTransformFromRefPose(self=@n3, BoneName=clavicle_r, BaseName=spine_05).ReturnValue_Rotation_Pitch, InRangeA=0, InRangeB=60)) -> (Output_Get:double)
  ]}
]
variables [
 LOD2 {type=boolean}
 Upperarm_L_rot {type=number}
 Upperarm_R_rot {type=number}
]
